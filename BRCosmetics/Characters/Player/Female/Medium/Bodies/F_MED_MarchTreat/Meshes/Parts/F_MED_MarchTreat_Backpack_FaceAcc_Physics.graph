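# Physics asset for the F_MED_MarchTreat FaceAcc part, setting up capsule bodies for left and right dangling lock chains.
# Each chain runs from an MSTR_lock root through two dyn_lock segments, with tight swing and very narrow twist limits.
# Lock roots and head constrain to neck_01, and neck_01 to spine_05, with free rotation.
# Capsules on head, neck, upper spine, clavicles and upper arms approximate the body; clavicles and arms have no constraints.
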PhysicsAsset /BRCosmetics/Characters/Player/Female/Medium/Bodies/F_MED_MarchTreat/Meshes/Parts/F_MED_MarchTreat_Backpack_FaceAcc_Physics
bodies {
 MSTR_lock_l [{shapeType=Capsule, length=0.505, radius=0.505, shapeName=MSTR_lock_l_capsule, rotation=<rot 0 -90 0>}]
 MSTR_lock_r [{shapeType=Capsule, length=0.505, radius=0.505, shapeName=MSTR_lock_r_capsule, rotation=<rot 0 -90 180>}]
 clavicle_l [{
   shapeType=Capsule
   length=8.7202
   radius=6.6438
   shapeName=clavicle_l_capsule
   center=<4.0734 -0.4074 -1.967>
   rotation=<rot -0.011 -89.9993 -85.9436>
  }]
 clavicle_r [{
   shapeType=Capsule
   length=8.853
   radius=6.8783
   shapeName=clavicle_r_capsule
   center=<-4.0734 0.4074 1.967>
   rotation=<rot -0.011 -89.9993 94.0564>
  }]
 dyn_lock_01_l [{shapeType=Capsule, length=5.0677, radius=1.6846, shapeName=dyn_lock_01_l_capsule, center=<4.0362 0 0>, rotation=<rot 0 97.775 -87.8731>}]
 dyn_lock_01_r [{shapeType=Capsule, length=5.264, radius=1.4878, shapeName=dyn_lock_01_r_capsule, center=<4 0 0>, rotation=<rot 0.5247 83.4985 92.0613>}]
 dyn_lock_02_l [{
   shapeType=Capsule
   length=4.7132
   radius=1.1455
   shapeName=dyn_lock_02_l_capsule
   center=<3.2547 0 0>
   rotation=<rot 1.2769 87.8968 -90.7762>
  }]
 dyn_lock_02_r [{
   shapeType=Capsule
   length=4.905
   radius=0.9536
   shapeName=dyn_lock_02_r_capsule
   center=<3.2477 -0.0012 0.064>
   rotation=<rot 1.277 87.8968 96.1896>
  }]
 head [
  {shapeType=Capsule, length=9.1982, radius=6.2959, center=<-3.2964 0 0>, rotation=<rot 90 0 0>}
  {
   shapeType=Capsule
   length=4.0211
   radius=8.2441
   shapeName=head_capsule
   center=<-0.2502 1.6678 0.0048>
   rotation=<rot 88.9044 -0.0001 -8.5342>
  }
 ]
 neck_01 [{shapeType=Capsule, length=7.1437, radius=4.1081, shapeName=neck_01_capsule, center=<0.6115 0.8027 0>, rotation=<rot 90 4.2382 0>}]
 spine_05 [{shapeType=Capsule, length=4, radius=10, shapeName=spine_05_capsule, center=<0.6464 0.8319 0>, rotation=<rot 0 -90 0>}]
 upperarm_l [{shapeType=Capsule, length=19.1798, radius=4.8625, shapeName=upperarm_l_capsule, center=<12.9575 0 0>, rotation=<rot 90 0 0>}]
 upperarm_r [{shapeType=Capsule, length=18.402, radius=4.5988, shapeName=upperarm_r_capsule, center=<-10.9731 0.7934 0>, rotation=<rot 90 180 0>}]
}
constraints [
 {bone1Name=dyn_lock_01_l, bone2Name=MSTR_lock_l, swing1LimitDegrees=12, swing2LimitDegrees=20, twistLimitDegrees=2}
 {bone1Name=dyn_lock_02_l, bone2Name=dyn_lock_01_l, swing1LimitDegrees=20, swing2LimitDegrees=20, twistLimitDegrees=2}
 {bone1Name=dyn_lock_01_r, bone2Name=MSTR_lock_r, swing1LimitDegrees=12, swing2LimitDegrees=25, twistLimitDegrees=2}
 {bone1Name=dyn_lock_02_r, bone2Name=dyn_lock_01_r, swing1LimitDegrees=12, swing2LimitDegrees=20, twistLimitDegrees=2}
 {
  bone1Name=head
  bone2Name=neck_01
  swing1LimitDegrees=45
  swing1Motion=Free
  swing2LimitDegrees=45
  swing2Motion=Free
  twistLimitDegrees=45
  twistMotion=Free
 }
 {
  bone1Name=MSTR_lock_l
  bone2Name=neck_01
  swing1LimitDegrees=45
  swing1Motion=Free
  swing2LimitDegrees=45
  swing2Motion=Free
  twistLimitDegrees=45
  twistMotion=Free
 }
 {
  bone1Name=MSTR_lock_r
  bone2Name=neck_01
  swing1LimitDegrees=45
  swing1Motion=Free
  swing2LimitDegrees=45
  swing2Motion=Free
  twistLimitDegrees=45
  twistMotion=Free
 }
 {
  bone1Name=neck_01
  bone2Name=spine_05
  swing1LimitDegrees=45
  swing1Motion=Free
  swing2LimitDegrees=45
  swing2Motion=Free
  twistLimitDegrees=45
  twistMotion=Free
 }
]
