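# Physics asset for the F_MED_MarchTreat female character's face accessory part, setting up physics for its hair and side locks.
# Center, left and right four-bone hair chains and two two-bone locks hang from MSTR anchor bodies through limited swing and twist constraints.
# Locked constraints cross-link left to center at the third segment and right to center at the fourth.
# Capsules on head, neck, spine, clavicles and upper arms provide body shapes around the hair.
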PhysicsAsset /BRCosmetics/Characters/Player/Female/Medium/Bodies/F_MED_MarchTreat/Meshes/Parts/F_MED_MarchTreat_FaceAcc_Physics
bodies {
 MSTR_hair_c [{shapeType=Capsule, length=0.505, radius=0.505, shapeName=MSTR_hair_c_capsule, rotation=<rot 0 -90 0>}]
 MSTR_hair_l [{shapeType=Capsule, length=0.505, radius=0.505, shapeName=MSTR_hair_l_capsule, rotation=<rot 0 -90 0>}]
 MSTR_hair_r [{shapeType=Capsule, length=0.505, radius=0.505, shapeName=MSTR_hair_r_capsule, rotation=<rot 0 -90 0>}]
 MSTR_lock_l [{shapeType=Capsule, length=0.505, radius=0.505, shapeName=MSTR_lock_l_capsule, rotation=<rot 0 -90 0>}]
 MSTR_lock_r [{shapeType=Capsule, length=0.505, radius=0.505, shapeName=MSTR_lock_r_capsule, rotation=<rot 0 -90 180>}]
 clavicle_l [{shapeType=Capsule, length=12, radius=5.5, shapeName=clavicle_l_capsule, center=<3.5 -1 -1>, rotation=<rot 80 -0.0003 -0.0003>}]
 clavicle_r [{shapeType=Capsule, length=12, radius=5.5, shapeName=clavicle_r_capsule, center=<-3.5 1 1>, rotation=<rot 80 -0.0003 -0.0004>}]
 dyn_hair_01_c [{shapeType=Capsule, length=6.2926, radius=4.3182, shapeName=dyn_hair_01_c_capsule, center=<6.231 0 0>, rotation=<rot 90 0 0>}]
 dyn_hair_01_l [{shapeType=Capsule, length=5.9075, radius=4.1566, shapeName=dyn_hair_01_l_capsule, center=<6 0 0>, rotation=<rot 90 0 0>}]
 dyn_hair_01_r [{shapeType=Capsule, length=5.9075, radius=4.1566, shapeName=dyn_hair_01_r_capsule, center=<6 0 0>, rotation=<rot 90 180 0>}]
 dyn_hair_02_c [{shapeType=Capsule, length=6.7568, radius=4.2901, shapeName=dyn_hair_02_c_capsule, center=<5.691 0 0>, rotation=<rot 90 0 0>}]
 dyn_hair_02_l [{
   shapeType=Capsule
   length=6.0618
   radius=4.3683
   shapeName=dyn_hair_02_l_capsule
   center=<4.9676 0.8465 0.2617>
   rotation=<rot 81.2315 -177.8282 180>
  }]
 dyn_hair_02_r [{shapeType=Capsule, length=6.0618, radius=4.3683, shapeName=dyn_hair_02_r_capsule, center=<4 0.8465 0.2617>, rotation=<rot 77.2512 2.172 180>}]
 dyn_hair_03_c [{shapeType=Capsule, length=6.3275, radius=6, shapeName=dyn_hair_03_c_capsule, center=<6 -3 0>, rotation=<rot 0 3.7642 0>}]
 dyn_hair_03_l [{shapeType=Capsule, length=8.3646, radius=6, shapeName=dyn_hair_03_l_capsule, center=<6 -3 2.2344>, rotation=<rot 18.6701 0.0007 25>}]
 dyn_hair_03_r [{shapeType=Capsule, length=8.2665, radius=6, shapeName=dyn_hair_03_r_capsule, center=<6 -3 -3.0981>, rotation=<rot -23.6192 3.5297 154.0754>}]
 dyn_hair_04_c [{shapeType=Capsule, length=11.1699, radius=4.8279, shapeName=dyn_hair_04_c_capsule, center=<8.2633 2.9557 0>, rotation=<rot 0 29.8986 0>}]
 dyn_hair_04_l [{
   shapeType=Capsule
   length=8.7554
   radius=4.2203
   shapeName=dyn_hair_04_l_capsule
   center=<7.7633 3.8758 1.6887>
   rotation=<rot 22.5299 42.1987 29.8962>
  }]
 dyn_hair_04_r [{
   shapeType=Capsule
   length=4.6895
   radius=4.094
   shapeName=dyn_hair_04_r_capsule
   center=<7.2653 3.4984 -5.0409>
   rotation=<rot -16.4026 -40.1608 69.9372>
  }]
 dyn_lock_01_l [{shapeType=Capsule, length=5.0677, radius=1.6846, shapeName=dyn_lock_01_l_capsule, center=<4.0362 0 0>, rotation=<rot 0 97.775 -87.8731>}]
 dyn_lock_01_r [{shapeType=Capsule, length=5.264, radius=1.4878, shapeName=dyn_lock_01_r_capsule, center=<4 0 0>, rotation=<rot 0.5247 83.4985 92.0613>}]
 dyn_lock_02_l [{shapeType=Capsule, length=4.7132, radius=1.25, shapeName=dyn_lock_02_l_capsule, center=<3.2547 0 0>, rotation=<rot 1.2769 87.8968 -90.7762>}]
 dyn_lock_02_r [{
   shapeType=Capsule
   length=4.905
   radius=1.25
   shapeName=dyn_lock_02_r_capsule
   center=<3.2477 -0.0012 0.064>
   rotation=<rot 1.277 87.8968 96.1896>
  }]
 head [
  {shapeType=Capsule, length=9.1982, radius=6.2959, center=<-3.2964 0 0>, rotation=<rot 90 0 0>}
  {
   shapeType=Capsule
   length=4.0211
   radius=8.2441
   shapeName=head_capsule
   center=<-0.2502 1.6678 0.0048>
   rotation=<rot 88.9044 -0.0001 -8.5342>
  }
 ]
 neck_01 [{shapeType=Capsule, length=7.1437, radius=4.1081, shapeName=neck_01_capsule, center=<0.6115 0.8027 0>, rotation=<rot 90 4.2382 0>}]
 spine_03 [{shapeType=Capsule, length=15, radius=10, shapeName=spine_04_capsule, center=<0.229 2 0>, rotation=<rot 90 0 10.0001>}]
 spine_05 [
  {shapeType=Capsule, length=12, radius=9.75, shapeName=spine_05_capsule, center=<-1.8049 0.5076 0>}
  {shapeType=Capsule, length=12, radius=7.75, shapeName=spine_05_capsule, center=<-5.7241 4.1842 0>}
 ]
 upperarm_l [{shapeType=Capsule, length=20, radius=4.5, shapeName=upperarm_l_capsule, center=<12 -0.5 0>, rotation=<rot 90 0 0>}]
 upperarm_r [{shapeType=Capsule, length=20, radius=5.5, shapeName=upperarm_r_capsule, center=<-12 0.5 0>, rotation=<rot 90 0 0>}]
}
constraints [
 {bone1Name=dyn_hair_01_c, bone2Name=MSTR_hair_c, swing1LimitDegrees=25, swing2LimitDegrees=10, twistLimitDegrees=0}
 {bone1Name=dyn_hair_02_c, bone2Name=dyn_hair_01_c, swing1LimitDegrees=25, swing2LimitDegrees=15, twistLimitDegrees=2}
 {bone1Name=dyn_hair_03_c, bone2Name=dyn_hair_02_c, swing1LimitDegrees=25, swing2LimitDegrees=15, twistLimitDegrees=2}
 {bone1Name=dyn_hair_01_l, bone2Name=MSTR_hair_l, swing1LimitDegrees=25, swing2LimitDegrees=20, twistLimitDegrees=0}
 {bone1Name=dyn_hair_02_l, bone2Name=dyn_hair_01_l, swing1LimitDegrees=25, swing2LimitDegrees=27.28, twistLimitDegrees=0}
 {bone1Name=dyn_hair_03_l, bone2Name=dyn_hair_02_l, swing1LimitDegrees=30, swing2LimitDegrees=10, twistLimitDegrees=2}
 {bone1Name=dyn_hair_01_r, bone2Name=MSTR_hair_r, swing1LimitDegrees=20, swing2LimitDegrees=20, twistLimitDegrees=0}
 {bone1Name=dyn_hair_02_r, bone2Name=dyn_hair_01_r, swing1LimitDegrees=25, swing2LimitDegrees=22.4, twistLimitDegrees=0}
 {bone1Name=dyn_hair_03_r, bone2Name=dyn_hair_02_r, swing1LimitDegrees=30, swing2LimitDegrees=15, twistLimitDegrees=10}
 {bone1Name=dyn_hair_04_r, bone2Name=dyn_hair_03_r, swing1LimitDegrees=25, swing2LimitDegrees=25, twistLimitDegrees=10}
 {bone1Name=dyn_hair_04_l, bone2Name=dyn_hair_03_l, swing1LimitDegrees=25, swing2LimitDegrees=25, twistLimitDegrees=2}
 {bone1Name=dyn_hair_04_c, bone2Name=dyn_hair_03_c, swing1LimitDegrees=30, swing2LimitDegrees=20, twistLimitDegrees=2}
 {
  bone1Name=dyn_hair_03_l
  bone2Name=dyn_hair_03_c
  swing1LimitDegrees=45
  swing1Motion=Locked
  swing2LimitDegrees=45
  swing2Motion=Locked
  twistLimitDegrees=45
  twistMotion=Locked
 }
 {bone1Name=dyn_lock_01_l, bone2Name=MSTR_lock_l, swing1LimitDegrees=12, swing2LimitDegrees=25, twistLimitDegrees=2}
 {bone1Name=dyn_lock_02_l, bone2Name=dyn_lock_01_l, swing1LimitDegrees=20, swing2LimitDegrees=20, twistLimitDegrees=2}
 {bone1Name=dyn_lock_01_r, bone2Name=MSTR_lock_r, swing1LimitDegrees=12, swing2LimitDegrees=25, twistLimitDegrees=2}
 {bone1Name=dyn_lock_02_r, bone2Name=dyn_lock_01_r, swing1LimitDegrees=20, swing2LimitDegrees=20, twistLimitDegrees=2}
 {
  bone1Name=dyn_hair_04_r
  bone2Name=dyn_hair_04_c
  swing1LimitDegrees=45
  swing1Motion=Locked
  swing2LimitDegrees=45
  swing2Motion=Locked
  twistLimitDegrees=45
  twistMotion=Locked
 }
 {
  bone1Name=head
  bone2Name=neck_01
  swing1LimitDegrees=45
  swing1Motion=Free
  swing2LimitDegrees=45
  swing2Motion=Free
  twistLimitDegrees=45
  twistMotion=Free
 }
 {
  bone1Name=neck_01
  bone2Name=spine_05
  swing1LimitDegrees=45
  swing1Motion=Free
  swing2LimitDegrees=45
  swing2Motion=Free
  twistLimitDegrees=45
  twistMotion=Free
 }
]
