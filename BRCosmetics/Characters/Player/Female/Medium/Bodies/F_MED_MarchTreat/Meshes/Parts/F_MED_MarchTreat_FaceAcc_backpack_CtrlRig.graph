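# Control Rig for the MarchTreat female face accessory (backpack variant) that applies corrective offsets to the left, center and right hair bone chains.
# A sequence runs corrections weighted by spherical pose readers on head, clavicles and upper arms, or by movement bools such as skydiving and parachute, often interpolated.
# Parent constraints blend some hair bones toward body bones, including while crouching.
# Sprint and tactical-sprint corrections exist but are disconnected from execution.
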ControlRigBlueprint /BRCosmetics/Characters/Player/Female/Medium/Bodies/F_MED_MarchTreat/Meshes/Parts/F_MED_MarchTreat_FaceAcc_backpack_CtrlRig
graphs [
 RigVMFunctionLibrary {links [], nodes []}
 RigVMFunction_Sequence_1_ContainedGraph {nodes [
   n1 = RigVMFunctionEntryNode()
   RigVMFunction_Sequence = &RigVMFunction_Sequence(ExecuteContext=@n1.ExecuteContext)
   n2 = &RigVMFunction_Sequence(ExecuteContext=@RigVMFunction_Sequence.B)
   n3 = &RigVMFunction_Sequence(ExecuteContext=@n2.B)
   n4 = &RigVMFunction_Sequence(ExecuteContext=@n3.B)
   n5 = &RigVMFunction_Sequence(ExecuteContext=@n4.B)
   n6 = &RigVMFunction_Sequence(ExecuteContext=@n5.B)
   n7 = &RigVMFunction_Sequence(ExecuteContext=@n6.B)
   n8 = &RigVMFunction_Sequence(ExecuteContext=@n7.B)
   n9 = &RigVMFunction_Sequence(ExecuteContext=@n8.B)
   RigVMFunctionReturnNode(A=@RigVMFunction_Sequence.A, B=@n2.A, C=@n3.A, D=@n4.A, E=@n5.A, F=@n6.A, G=@n7.A, H=@n8.A, I=@n9.A, J=@n9.B)
  ]}
 RigVMModel {nodes [
   RigUnit_BeginExecution = &RigUnit_BeginExecution()
   RigVMFunction_Sequence = RigVMAggregateNode(ExecuteContext=@RigUnit_BeginExecution.ExecutePin)
   n1 = &RigUnit_SphericalPoseReader(
    ActiveRegionScaleFactors=(PositiveWidth=1,NegativeWidth=1,PositiveHeight=1,NegativeHeight=1)
    ActiveRegionSize=0.1
    Debug=(bDrawDebug=false,bDraw2D=False,bDrawLocalAxes=False,DebugScale=25,DebugSegments=20,DebugThickness=0.25)
    DriverAxis=(X=-1,Y=0,Z=0)
    DriverItem=(Type=Bone,Name="head")
    FalloffRegionScaleFactors=(PositiveWidth=1,NegativeWidth=1,PositiveHeight=1,NegativeHeight=1)
    FalloffSize=0.2
    FlipHeightScaling=False
    FlipWidthScaling=False
    OptionalParentItem=(Type=Bone,Name="None")
    RotationOffset=(X=90,Y=0,Z=125)
    ExecutePin=@RigVMFunction_Sequence.A
   )
   > /Script/ControlRig.RigUnit_SetRelativeTransformForItem(
    Child=(Type=Bone,Name="MSTR_hair_r")
    Parent=(Type=Bone,Name="MSTR_hair_r")
    Value=(Rotation=(X=0,Y=0,Z=0,W=1),Translation=(X=-1.5,Y=0,Z=0),Scale3D=(X=0.95,Y=0.95,Z=0.95))
    bParentInitial=False
    bPropagateToChildren=True
    Weight=@n1.OutputParam
   ) {
    is=struct RigVMUnitNode
    template="Set Relative Transform::Execute(in Child,in Parent,in bParentInitial,in Value,in Weight,in bPropagateToChildren)"
   }
   > &RigUnit_SetRelativeTransformForItem(
    Child=(Type=Bone,Name="MSTR_hair_c")
    Parent=(Type=Bone,Name="MSTR_hair_c")
    Value=(Rotation=(X=0,Y=0,Z=0,W=1),Translation=(X=-1.5,Y=0,Z=0),Scale3D=(X=0.95,Y=0.95,Z=0.95))
    bParentInitial=False
    bPropagateToChildren=True
    Weight=@n1.OutputParam
   ) {template="Set Relative Transform::Execute(in Child,in Parent,in bParentInitial,in Value,in Weight,in bPropagateToChildren)"}
   > &RigUnit_SetRelativeTransformForItem(
    Child=(Type=Bone,Name="MSTR_hair_l")
    Parent=(Type=Bone,Name="MSTR_hair_l")
    Value=(Rotation=(X=0,Y=0,Z=0,W=1),Translation=(X=-1.5,Y=0,Z=0),Scale3D=(X=0.95,Y=0.95,Z=0.95))
    bParentInitial=False
    bPropagateToChildren=True
    Weight=@n1.OutputParam
   ) {template="Set Relative Transform::Execute(in Child,in Parent,in bParentInitial,in Value,in Weight,in bPropagateToChildren)"}
   n2 = &RigVMFunction_AlphaInterp(
    InterpSpeedDecreasing=9
    InterpSpeedIncreasing=10
    Scale=1
    bClampResult=false
    Value=&RigVMFunction_MathBoolToFloat(Value=RigVMVariableNode() {variable=Sky diving}.Value).Result
   )
   &RigUnit_SetRelativeTransformForItem(
    Child=(Type=Bone,Name="MSTR_hair_l")
    Parent=(Type=Bone,Name="MSTR_hair_l")
    Value=(Rotation=(X=0.0532,Y=-0.1619,Z=0.1011,W=-0.9802),Translation=(X=0,Y=0,Z=0),Scale3D=(X=1,Y=1,Z=1))
    bParentInitial=False
    bPropagateToChildren=True
    Weight=@n2.Result
    ExecutePin=@RigVMFunction_Sequence.B
   ) {template="Set Relative Transform::Execute(in Child,in Parent,in bParentInitial,in Value,in Weight,in bPropagateToChildren)"}
   > &RigUnit_SetRelativeTransformForItem(
    Child=(Type=Bone,Name="dyn_hair_03_l")
    Parent=(Type=Bone,Name="dyn_hair_03_l")
    Value=(Rotation=(X=-0.011,Y=-0.0308,Z=0.0294,W=-0.999),Translation=(X=0,Y=0,Z=0),Scale3D=(X=1,Y=1,Z=1))
    bParentInitial=False
    bPropagateToChildren=True
    Weight=@n2.Result
   ) {template="Set Relative Transform::Execute(in Child,in Parent,in bParentInitial,in Value,in Weight,in bPropagateToChildren)"}
   > &RigUnit_SetRelativeTransformForItem(
    Child=(Type=Bone,Name="MSTR_hair_c")
    Parent=(Type=Bone,Name="MSTR_hair_c")
    Value=(Rotation=(X=0,Y=0,Z=0.0839,W=-0.9965),Translation=(X=0,Y=0,Z=0),Scale3D=(X=1,Y=1,Z=1))
    bParentInitial=False
    bPropagateToChildren=True
    Weight=@n2.Result
   ) {template="Set Relative Transform::Execute(in Child,in Parent,in bParentInitial,in Value,in Weight,in bPropagateToChildren)"}
   > &RigUnit_SetRelativeTransformForItem(
    Child=(Type=Bone,Name="MSTR_hair_r")
    Parent=(Type=Bone,Name="MSTR_hair_r")
    Value=(Rotation=(X=0.0702,Y=0,Z=0,W=-0.9975),Translation=(X=0,Y=0,Z=0),Scale3D=(X=1,Y=1,Z=1))
    bParentInitial=False
    bPropagateToChildren=True
    Weight=@n2.Result
   ) {template="Set Relative Transform::Execute(in Child,in Parent,in bParentInitial,in Value,in Weight,in bPropagateToChildren)"}
   > &RigUnit_SetRelativeTransformForItem(
    Child=(Type=Bone,Name="dyn_hair_02_r")
    Parent=(Type=Bone,Name="dyn_hair_02_r")
    Value=(Rotation=(X=0.0266,Y=-0.1232,Z=0.2095,W=0.9696),Translation=(X=0,Y=0,Z=3),Scale3D=(X=1,Y=1,Z=1))
    bParentInitial=False
    bPropagateToChildren=True
    Weight=@n2.Result
   ) {template="Set Relative Transform::Execute(in Child,in Parent,in bParentInitial,in Value,in Weight,in bPropagateToChildren)"}
   > &RigUnit_SetRelativeTransformForItem(
    Child=(Type=Bone,Name="dyn_hair_03_r")
    Parent=(Type=Bone,Name="dyn_hair_03_r")
    Value=(Rotation=(X=-0.1793,Y=-0.1161,Z=-0.2085,W=0.9544),Translation=(X=0,Y=-2,Z=0),Scale3D=(X=1,Y=1,Z=1))
    bParentInitial=False
    bPropagateToChildren=True
    Weight=@n2.Result
   ) {template="Set Relative Transform::Execute(in Child,in Parent,in bParentInitial,in Value,in Weight,in bPropagateToChildren)"}
   n3 = &RigVMFunction_ControlFlowBranch(ExecuteContext=@RigVMFunction_Sequence.C, Condition=RigVMVariableNode() {variable=Sky dive floating}.Value)
   n4 = &RigUnit_SphericalPoseReader(
    ActiveRegionScaleFactors=(PositiveWidth=1,NegativeWidth=1,PositiveHeight=1.7,NegativeHeight=1.7)
    ActiveRegionSize=0.1
    Debug=(bDrawDebug=false,bDraw2D=false,DebugScale=25,DebugSegments=20,DebugThickness=0.25)
    DriverAxis=(X=0,Y=-1,Z=0)
    DriverItem=(Type=Bone,Name="head")
    FalloffRegionScaleFactors=(PositiveWidth=2,NegativeWidth=2,PositiveHeight=2,NegativeHeight=2)
    FalloffSize=0.2
    FlipWidthScaling=false
    OptionalParentItem=(Type=Bone,Name="neck_01")
    RotationOffset=(X=-90,Y=0,Z=-45)
    ExecutePin=@n3.False
   )
   > RigVMVariableNode(Value=@n4.OutputParam) {variable=Head up alpha}
   > &RigUnit_SetRelativeTransformForItem(
    Child=(Type=Bone,Name="MSTR_hair_r")
    Parent=(Type=Bone,Name="MSTR_hair_r")
    Value=(Rotation=(X=-0.2432,Y=0.0885,Z=-0.3304,W=0.9077),Translation=(X=0,Y=0,Z=0),Scale3D=(X=1,Y=1,Z=1))
    bParentInitial=False
    bPropagateToChildren=True
    Weight=@n4.OutputParam
   ) {template="Set Relative Transform::Execute(in Child,in Parent,in bParentInitial,in Value,in Weight,in bPropagateToChildren)"}
   > &RigUnit_SetRelativeTransformForItem(
    Child=(Type=Bone,Name="dyn_hair_02_r")
    Parent=(Type=Bone,Name="dyn_hair_02_r")
    Value=(Rotation=(X=0,Y=0,Z=0.5736,W=0.8192),Translation=(X=-1,Y=-3,Z=0),Scale3D=(X=1,Y=1,Z=1))
    bParentInitial=False
    bPropagateToChildren=True
    Weight=@n4.OutputParam
   ) {template="Set Relative Transform::Execute(in Child,in Parent,in bParentInitial,in Value,in Weight,in bPropagateToChildren)"}
   > &RigUnit_SetRelativeTransformForItem(
    Child=(Type=Bone,Name="dyn_hair_03_r")
    Parent=(Type=Bone,Name="dyn_hair_03_r")
    Value=(Rotation=(X=0,Y=0,Z=-0.0872,W=0.9962),Translation=(X=-1,Y=0,Z=0),Scale3D=(X=1,Y=1,Z=1))
    bParentInitial=False
    bPropagateToChildren=True
    Weight=@n4.OutputParam
   ) {template="Set Relative Transform::Execute(in Child,in Parent,in bParentInitial,in Value,in Weight,in bPropagateToChildren)"}
   > &RigUnit_SetRelativeTransformForItem(
    Child=(Type=Bone,Name="MSTR_hair_l")
    Parent=(Type=Bone,Name="MSTR_hair_l")
    Value=(Rotation=(X=0,Y=0,Z=-0.4226,W=0.9063),Translation=(X=0,Y=0,Z=0),Scale3D=(X=1,Y=1,Z=1))
    bParentInitial=False
    bPropagateToChildren=True
    Weight=@n4.OutputParam
   ) {template="Set Relative Transform::Execute(in Child,in Parent,in bParentInitial,in Value,in Weight,in bPropagateToChildren)"}
   > &RigUnit_SetRelativeTransformForItem(
    Child=(Type=Bone,Name="dyn_hair_02_l")
    Parent=(Type=Bone,Name="dyn_hair_02_l")
    Value=(Rotation=(X=0,Y=0,Z=0.2588,W=0.9659),Translation=(X=0,Y=0,Z=0),Scale3D=(X=1,Y=1,Z=1))
    bParentInitial=False
    bPropagateToChildren=True
    Weight=@n4.OutputParam
   ) {template="Set Relative Transform::Execute(in Child,in Parent,in bParentInitial,in Value,in Weight,in bPropagateToChildren)"}
   > &RigUnit_SetRelativeTransformForItem(
    Child=(Type=Bone,Name="dyn_hair_03_l")
    Parent=(Type=Bone,Name="dyn_hair_03_l")
    Value=(Rotation=(X=0,Y=0,Z=0.4226,W=0.9063),Translation=(X=0,Y=0,Z=0),Scale3D=(X=1,Y=1,Z=1))
    bParentInitial=False
    bPropagateToChildren=True
    Weight=@n4.OutputParam
   ) {template="Set Relative Transform::Execute(in Child,in Parent,in bParentInitial,in Value,in Weight,in bPropagateToChildren)"}
   > &RigUnit_SetRelativeTransformForItem(
    Child=(Type=Bone,Name="dyn_hair_03_l")
    Parent=(Type=Bone,Name="dyn_hair_03_l")
    Value=(Rotation=(X=0.1555,Y=-0.1146,Z=-0.1555,W=0.9688),Translation=(X=0,Y=0,Z=1),Scale3D=(X=1,Y=1,Z=1))
    bParentInitial=False
    bPropagateToChildren=True
    Weight=@n4.OutputParam
   ) {template="Set Relative Transform::Execute(in Child,in Parent,in bParentInitial,in Value,in Weight,in bPropagateToChildren)"}
   > &RigUnit_SetRelativeTransformForItem(
    Child=(Type=Bone,Name="dyn_hair_04_r")
    Parent=(Type=Bone,Name="dyn_hair_04_r")
    Value=(Rotation=(X=0,Y=0,Z=-0.3007,W=0.9537),Translation=(X=0,Y=0,Z=0),Scale3D=(X=1,Y=1,Z=1))
    bParentInitial=False
    bPropagateToChildren=True
    Weight=@n4.OutputParam
   ) {template="Set Relative Transform::Execute(in Child,in Parent,in bParentInitial,in Value,in Weight,in bPropagateToChildren)"}
   > &RigUnit_ParentConstraint(
    AdvancedSettings=(RotationOrderForFilter=XZY)
    Child=(Type=Bone,Name="dyn_hair_02_l")
    Parents="((Item=(Type=Bone,Name=\"head\"),Weight=0.0222),(Item=(Type=Bone,Name=\"spine_03\"),Weight=0.0309),(Item=(Type=Bone,Name=\"clavicle_l\"),Weight=0.0485),(Item=(Type=Bone,Name=\"neck_01\"),Weight=0.1314),(Item=(Type=Bone,Name=\"neck_02\"),Weight=0.14),(Item=(Type=Bone ...(truncated)"
    Weight=1
   )
   > &RigUnit_ParentConstraint(
    AdvancedSettings=(RotationOrderForFilter=XZY)
    Child=(Type=Bone,Name="dyn_hair_03_l")
    Parents=((Item=(Type=Bone,Name="clavicle_r"),Weight=0.0003),(Item=(Type=Bone,Name="spine_05"),Weight=0.4532),(Item=(Type=Bone,Name="clavicle_l"),Weight=0.5465))
    Weight=1
   )
   n5 = &RigUnit_SphericalPoseReader(
    ActiveRegionScaleFactors=(PositiveWidth=1,NegativeWidth=1,PositiveHeight=1,NegativeHeight=1)
    ActiveRegionSize=0.1
    Debug=(bDrawDebug=false,bDraw2D=False,bDrawLocalAxes=False,DebugScale=25,DebugSegments=20,DebugThickness=0.25)
    DriverAxis=(X=-1,Y=0,Z=0)
    DriverItem=(Type=Bone,Name="upperarm_r")
    FalloffRegionScaleFactors=(PositiveWidth=1,NegativeWidth=1,PositiveHeight=1,NegativeHeight=1)
    FalloffSize=0.23
    FlipHeightScaling=False
    FlipWidthScaling=False
    OptionalParentItem=(Type=Bone,Name="None")
    RotationOffset=(X=-120,Y=0,Z=29)
    ExecutePin=@RigVMFunction_Sequence.D
   )
   n6 = &RigVMFunction_AlphaInterp(InterpSpeedDecreasing=10, InterpSpeedIncreasing=10, Scale=1, bClampResult=false, Value=@n5.OutputParam)
   > RigUnit_SetRelativeTransformForItem = &RigUnit_SetRelativeTransformForItem(
    Child=(Type=Bone,Name="MSTR_hair_r")
    Parent=(Type=Bone,Name="MSTR_hair_r")
    Value=(Rotation=(X=0.0361,Y=0.1699,Z=-0.2048,W=0.9633),Translation=(X=0,Y=0,Z=0),Scale3D=(X=1,Y=1,Z=1))
    bParentInitial=False
    bPropagateToChildren=True
    Weight=@n6.Result
   ) {template="Set Relative Transform::Execute(in Child,in Parent,in bParentInitial,in Value,in Weight,in bPropagateToChildren)"}
   > &RigUnit_SetRelativeTransformForItem(
    Child=(Type=Bone,Name="dyn_hair_02_r")
    Parent=(Type=Bone,Name="dyn_hair_02_r")
    Value=(Rotation=(X=0,Y=0,Z=0,W=1),Translation=(X=0,Y=-2,Z=0),Scale3D=(X=1,Y=1,Z=1))
    bParentInitial=False
    bPropagateToChildren=True
    Weight=@n6.Result
   ) {template="Set Relative Transform::Execute(in Child,in Parent,in bParentInitial,in Value,in Weight,in bPropagateToChildren)"}
   > &RigUnit_SetRelativeTransformForItem(
    Child=(Type=Bone,Name="dyn_hair_03_r")
    Parent=(Type=Bone,Name="dyn_hair_03_r")
    Value=(Rotation=(X=0.117,Y=-0.3214,Z=0.3214,W=0.883),Translation=(X=0,Y=5,Z=0),Scale3D=(X=1,Y=1,Z=1))
    bParentInitial=False
    bPropagateToChildren=True
    Weight=@n6.Result
   ) {template="Set Relative Transform::Execute(in Child,in Parent,in bParentInitial,in Value,in Weight,in bPropagateToChildren)"}
   RigVMFunction_ControlFlowBranch = &RigVMFunction_ControlFlowBranch(ExecuteContext=@RigVMFunction_Sequence.E, Condition=RigVMVariableNode() {variable=Sky dive floating}.Value)
   n7 = &RigUnit_SphericalPoseReader(
    ActiveRegionScaleFactors=(PositiveWidth=1,NegativeWidth=1,PositiveHeight=1,NegativeHeight=1)
    ActiveRegionSize=0.1
    Debug=(bDrawDebug=false,bDraw2D=False,bDrawLocalAxes=False,DebugScale=25,DebugSegments=20,DebugThickness=0.25)
    DriverAxis=(X=1,Y=0,Z=0)
    DriverItem=(Type=Bone,Name="clavicle_l")
    FalloffRegionScaleFactors=(PositiveWidth=1,NegativeWidth=1,PositiveHeight=1,NegativeHeight=1)
    FalloffSize=0.2
    FlipHeightScaling=False
    FlipWidthScaling=False
    OptionalParentItem=(Type=Bone,Name="None")
    RotationOffset=(X=75,Y=0,Z=-125)
    ExecutePin=@RigVMFunction_ControlFlowBranch.False
   )
   > RigVMVariableNode(Value=@n7.OutputParam) {variable=Clavicle L Back Alpha}
   n8 = &RigVMFunction_AlphaInterp(InterpSpeedDecreasing=10, InterpSpeedIncreasing=10, Scale=1, bClampResult=false, Value=@n7.OutputParam)
   > &RigUnit_SetRelativeTransformForItem(
    Child=(Type=Bone,Name="MSTR_hair_l")
    Parent=(Type=Bone,Name="MSTR_hair_l")
    Value=(Rotation=(X=0.0061,Y=0.0344,Z=-0.1735,W=0.9842),Translation=(X=0,Y=0,Z=0),Scale3D=(X=1,Y=1,Z=1))
    bParentInitial=False
    bPropagateToChildren=True
    Weight=@n8.Result
   ) {template="Set Relative Transform::Execute(in Child,in Parent,in bParentInitial,in Value,in Weight,in bPropagateToChildren)"}
   > &RigUnit_SetRelativeTransformForItem(
    Child=(Type=Bone,Name="dyn_hair_02_l")
    Parent=(Type=Bone,Name="dyn_hair_02_l")
    Value=(Rotation=(X=0,Y=0.0872,Z=0,W=0.9962),Translation=(X=0,Y=0,Z=0),Scale3D=(X=1,Y=1,Z=1))
    bParentInitial=False
    bPropagateToChildren=True
    Weight=@n8.Result
   ) {template="Set Relative Transform::Execute(in Child,in Parent,in bParentInitial,in Value,in Weight,in bPropagateToChildren)"}
   > &RigUnit_SetRelativeTransformForItem(
    Child=(Type=Bone,Name="dyn_hair_03_l")
    Parent=(Type=Bone,Name="dyn_hair_03_l")
    Value=(Rotation=(X=0,Y=0,Z=0.1305,W=0.9914),Translation=(X=0,Y=0,Z=0),Scale3D=(X=1,Y=1,Z=1))
    bParentInitial=False
    bPropagateToChildren=True
    Weight=@n8.Result
   ) {template="Set Relative Transform::Execute(in Child,in Parent,in bParentInitial,in Value,in Weight,in bPropagateToChildren)"}
   > &RigUnit_SetRelativeTransformForItem(
    Child=(Type=Bone,Name="MSTR_hair_c")
    Parent=(Type=Bone,Name="MSTR_hair_c")
    Value=(Rotation=(X=0,Y=0,Z=0,W=1),Translation=(X=0,Y=0,Z=0),Scale3D=(X=1,Y=1,Z=1))
    bParentInitial=False
    bPropagateToChildren=True
    Weight=@n8.Result
   ) {template="Set Relative Transform::Execute(in Child,in Parent,in bParentInitial,in Value,in Weight,in bPropagateToChildren)"}
   > n9 = &RigUnit_SphericalPoseReader(
    ActiveRegionScaleFactors=(PositiveWidth=1,NegativeWidth=1,PositiveHeight=1.5,NegativeHeight=1.5)
    ActiveRegionSize=0.1
    Debug=(bDrawDebug=false,bDraw2D=False,bDrawLocalAxes=False,DebugScale=25,DebugSegments=20,DebugThickness=0.25)
    DriverAxis=(X=-1,Y=0,Z=0)
    DriverItem=(Type=Bone,Name="clavicle_r")
    FalloffRegionScaleFactors=(PositiveWidth=2.5,NegativeWidth=2.5,PositiveHeight=4,NegativeHeight=4)
    FalloffSize=0.1796
    FlipHeightScaling=False
    FlipWidthScaling=False
    OptionalParentItem=(Type=Bone,Name="None")
    RotationOffset=(X=104.09,Y=0,Z=57.48)
   )
   > n10 = &RigUnit_SphericalPoseReader(
    ActiveRegionScaleFactors=(PositiveWidth=1,NegativeWidth=1,PositiveHeight=1,NegativeHeight=1)
    ActiveRegionSize=0.1
    Debug=(bDrawDebug=false,bDraw2D=False,bDrawLocalAxes=False,DebugScale=25,DebugSegments=20,DebugThickness=0.25)
    DriverAxis=(X=1,Y=0,Z=0)
    DriverItem=(Type=Bone,Name="upperarm_l")
    FalloffRegionScaleFactors=(PositiveWidth=1,NegativeWidth=1,PositiveHeight=1,NegativeHeight=1)
    FalloffSize=0.2
    FlipHeightScaling=False
    FlipWidthScaling=False
    OptionalParentItem=(Type=Bone,Name="None")
    RotationOffset=(X=107.61,Y=0,Z=-26.09)
   )
   > RigVMVariableNode(Value=0) {variable=Clavicle R Back Alpha}
   > &RigUnit_SphericalPoseReader(
    ActiveRegionScaleFactors=(PositiveWidth=1,NegativeWidth=1,PositiveHeight=1,NegativeHeight=1)
    ActiveRegionSize=0.1
    Debug=(bDrawDebug=false,bDraw2D=False,bDrawLocalAxes=False,DebugScale=25,DebugSegments=20,DebugThickness=0.25)
    DriverAxis=(X=-1,Y=0,Z=0)
    DriverItem=(Type=Bone,Name="upperarm_r")
    FalloffRegionScaleFactors=(PositiveWidth=1,NegativeWidth=1,PositiveHeight=1,NegativeHeight=1)
    FalloffSize=0.25
    FlipHeightScaling=False
    FlipWidthScaling=false
    OptionalParentItem=(Type=Bone,Name="None")
    RotationOffset=(X=-10,Y=90,Z=0)
   )
   n11 = &RigVMFunction_AlphaInterp(
    InterpSpeedDecreasing=10
    InterpSpeedIncreasing=10
    Scale=1
    bClampResult=false
    Value=&RigVMFunction_MathDoubleMul(A=@n9.OutputParam, B=@n10.OutputParam).Result
   )
   > &RigUnit_SetRelativeTransformForItem(
    Child=(Type=Bone,Name="MSTR_hair_r")
    Parent=(Type=Bone,Name="MSTR_hair_r")
    Value=(Rotation=(X=0,Y=0,Z=0,W=1),Translation=(X=-2,Y=0,Z=0),Scale3D=(X=1,Y=1,Z=1))
    bParentInitial=False
    bPropagateToChildren=True
    Weight=@n11.Result
   ) {template="Set Relative Transform::Execute(in Child,in Parent,in bParentInitial,in Value,in Weight,in bPropagateToChildren)"}
   RigVMFunction_MathBoolToFloat = &RigVMFunction_MathBoolToFloat(Value=RigVMVariableNode() {variable=parachute opened}.Value)
   n12 = &RigVMFunction_AlphaInterp(InterpSpeedDecreasing=10, InterpSpeedIncreasing=10, Scale=1, bClampResult=false, Value=@RigVMFunction_MathBoolToFloat.Result)
   &RigUnit_SetRelativeTransformForItem(
    Child=(Type=Bone,Name="MSTR_hair_r")
    Parent=(Type=Bone,Name="MSTR_hair_r")
    Value=(Rotation=(X=0,Y=0,Z=-0.2588,W=0.9659),Translation=(X=0,Y=0,Z=0),Scale3D=(X=1,Y=1,Z=1))
    bParentInitial=False
    bPropagateToChildren=True
    Weight=@n12.Result
    ExecutePin=@RigVMFunction_Sequence.F
   ) {template="Set Relative Transform::Execute(in Child,in Parent,in bParentInitial,in Value,in Weight,in bPropagateToChildren)"}
   > &RigUnit_SetRelativeTransformForItem(
    Child=(Type=Bone,Name="dyn_hair_02_l")
    Parent=(Type=Bone,Name="dyn_hair_02_l")
    Value=(Rotation=(X=0,Y=0,Z=0.2588,W=0.9659),Translation=(X=0,Y=0,Z=0),Scale3D=(X=1,Y=1,Z=1))
    bParentInitial=False
    bPropagateToChildren=True
    Weight=@n12.Result
   ) {template="Set Relative Transform::Execute(in Child,in Parent,in bParentInitial,in Value,in Weight,in bPropagateToChildren)"}
   > &RigUnit_SetRelativeTransformForItem(
    Child=(Type=Bone,Name="MSTR_hair_l")
    Parent=(Type=Bone,Name="MSTR_hair_l")
    Value=(Rotation=(X=0,Y=0,Z=-0.2588,W=0.9659),Translation=(X=0,Y=0,Z=0),Scale3D=(X=1,Y=1,Z=1))
    bParentInitial=False
    bPropagateToChildren=True
    Weight=@n12.Result
   ) {template="Set Relative Transform::Execute(in Child,in Parent,in bParentInitial,in Value,in Weight,in bPropagateToChildren)"}
   > &RigUnit_SetRelativeTransformForItem(
    Child=(Type=Bone,Name="dyn_hair_02_r")
    Parent=(Type=Bone,Name="dyn_hair_02_r")
    Value=(Rotation=(X=0,Y=0,Z=0.2588,W=0.9659),Translation=(X=0,Y=0,Z=0),Scale3D=(X=1,Y=1,Z=1))
    bParentInitial=False
    bPropagateToChildren=True
    Weight=@n12.Result
   ) {template="Set Relative Transform::Execute(in Child,in Parent,in bParentInitial,in Value,in Weight,in bPropagateToChildren)"}
   n13 = &RigVMFunction_AlphaInterp(
    InterpSpeedDecreasing=10
    InterpSpeedIncreasing=10
    Scale=1
    bClampResult=false
    Value=&RigVMFunction_MathBoolToFloat(Value=RigVMVariableNode() {variable=Sky dive floating}.Value).Result
   )
   &RigUnit_SetRelativeTransformForItem(
    Child=(Type=Bone,Name="dyn_hair_02_l")
    Parent=(Type=Bone,Name="dyn_hair_02_l")
    Value=(Rotation=(X=0.3616,Y=0,Z=0,W=0.9323),Translation=(X=1.8461,Y=-1.7698,Z=2.155),Scale3D=(X=1,Y=1,Z=1))
    bParentInitial=False
    bPropagateToChildren=True
    Weight=@n13.Result
    ExecutePin=@RigVMFunction_Sequence.G
   ) {template="Set Relative Transform::Execute(in Child,in Parent,in bParentInitial,in Value,in Weight,in bPropagateToChildren)"}
   > &RigUnit_SetRelativeTransformForItem(
    Child=(Type=Bone,Name="dyn_hair_03_l")
    Parent=(Type=Bone,Name="dyn_hair_03_l")
    Value=(Rotation=(X=-0.0192,Y=-0.1951,Z=0.0154,W=-0.9805),Translation=(X=0,Y=0.0779,Z=2.7827),Scale3D=(X=1,Y=1,Z=1))
    bParentInitial=False
    bPropagateToChildren=True
    Weight=@n13.Result
   ) {template="Set Relative Transform::Execute(in Child,in Parent,in bParentInitial,in Value,in Weight,in bPropagateToChildren)"}
   > &RigUnit_SetRelativeTransformForItem(
    Child=(Type=Bone,Name="dyn_hair_01_r")
    Parent=(Type=Bone,Name="dyn_hair_01_r")
    Value=(Rotation=(X=-0.0893,Y=0.3676,Z=-0.222,W=0.8987),Translation=(X=0,Y=-0.8112,Z=1.2575),Scale3D=(X=1,Y=1,Z=1))
    bParentInitial=False
    bPropagateToChildren=True
    Weight=@n13.Result
   ) {template="Set Relative Transform::Execute(in Child,in Parent,in bParentInitial,in Value,in Weight,in bPropagateToChildren)"}
   > &RigUnit_SetRelativeTransformForItem(
    Child=(Type=Bone,Name="dyn_hair_02_r")
    Parent=(Type=Bone,Name="dyn_hair_02_r")
    Value=(Rotation=(X=0.0117,Y=-0.0204,Z=0.259,W=-0.9656),Translation=(X=1.2135,Y=-2.4125,Z=1.3923),Scale3D=(X=1,Y=1,Z=1))
    bParentInitial=False
    bPropagateToChildren=True
    Weight=@n13.Result
   ) {template="Set Relative Transform::Execute(in Child,in Parent,in bParentInitial,in Value,in Weight,in bPropagateToChildren)"}
   > &RigUnit_SetRelativeTransformForItem(
    Child=(Type=Bone,Name="dyn_hair_03_r")
    Parent=(Type=Bone,Name="dyn_hair_03_r")
    Value=(Rotation=(X=-0.33,Y=0.2452,Z=-0.3472,W=-0.8429),Translation=(X=0,Y=4,Z=0),Scale3D=(X=1,Y=1,Z=1))
    bParentInitial=False
    bPropagateToChildren=True
    Weight=@n13.Result
   ) {template="Set Relative Transform::Execute(in Child,in Parent,in bParentInitial,in Value,in Weight,in bPropagateToChildren)"}
   n14 = &RigUnit_SphericalPoseReader(
    ActiveRegionScaleFactors=(PositiveWidth=1,NegativeWidth=0.5,PositiveHeight=1,NegativeHeight=1)
    ActiveRegionSize=0.1
    Debug=(bDrawDebug=false,DebugScale=25,DebugSegments=20,DebugThickness=0.25)
    DriverAxis=(X=0,Y=-1,Z=0)
    DriverItem=(Type=Bone,Name="head")
    FalloffRegionScaleFactors=(PositiveWidth=1,NegativeWidth=0.5,PositiveHeight=1,NegativeHeight=1)
    FalloffSize=0.2
    OptionalParentItem=(Type=Bone,Name="neck_02")
    RotationOffset=(X=-55,Y=0,Z=-10)
    ExecutePin=@RigVMFunction_Sequence.H
   )
   > &RigUnit_SetRelativeTransformForItem(
    Child=(Type=Bone,Name="MSTR_hair_r")
    Parent=(Type=Bone,Name="MSTR_hair_r")
    Value=(Rotation=(X=0,Y=0,Z=-0.0872,W=0.9962),Translation=(X=0,Y=0,Z=0),Scale3D=(X=1,Y=1,Z=1))
    bParentInitial=False
    bPropagateToChildren=True
    Weight=@n14.OutputParam
   ) {template="Set Relative Transform::Execute(in Child,in Parent,in bParentInitial,in Value,in Weight,in bPropagateToChildren)"}
   n15 = &RigVMFunction_ControlFlowBranch(ExecuteContext=@RigVMFunction_Sequence.I, Condition=RigVMVariableNode() {variable=Swim}.Value)
   n16 = &RigUnit_SphericalPoseReader(
    ActiveRegionScaleFactors=(PositiveWidth=1.4,NegativeWidth=1,PositiveHeight=1,NegativeHeight=1)
    ActiveRegionSize=0.1
    Debug=(bDrawDebug=true,bDraw2D=False,bDrawLocalAxes=False,DebugScale=25,DebugSegments=20,DebugThickness=0.25)
    DriverAxis=(X=-1,Y=0,Z=0)
    DriverItem=(Type=Bone,Name="upperarm_r")
    FalloffRegionScaleFactors=(PositiveWidth=2,NegativeWidth=1.4,PositiveHeight=1,NegativeHeight=1)
    FalloffSize=0.2
    FlipHeightScaling=False
    FlipWidthScaling=False
    OptionalParentItem=(Type=Bone,Name="None")
    RotationOffset=(X=-7.48,Y=116.16,Z=0)
    ExecutePin=@n15.True
   )
   > &RigUnit_SetRelativeTransformForItem(
    Child=(Type=Bone,Name="MSTR_hair_r")
    Parent=(Type=Bone,Name="MSTR_hair_r")
    Value=(Rotation=(X=-0.173,Y=0.0151,Z=-0.0858,W=0.9811),Translation=(X=0,Y=-2,Z=1),Scale3D=(X=1,Y=1,Z=1))
    bParentInitial=False
    bPropagateToChildren=True
    Weight=@n16.OutputParam
   ) {template="Set Relative Transform::Execute(in Child,in Parent,in bParentInitial,in Value,in Weight,in bPropagateToChildren)"}
   n17 = &RigVMFunction_ControlFlowBranch(ExecuteContext=@RigVMFunction_Sequence.J, Condition=RigVMVariableNode() {variable=crouching}.Value)
   &RigUnit_ParentConstraint(
    AdvancedSettings=(InterpolationType=Average,RotationOrderForFilter=XZY)
    Child=(Type=Bone,Name="dyn_hair_02_r")
    Parents=((Item=(Type=Bone,Name="spine_05"),Weight=0.5),(Item=(Type=Bone,Name="clavicle_r"),Weight=0.5))
    Weight=1
    ExecutePin=@n17.True
   )
   > &RigUnit_ParentConstraint(
    AdvancedSettings=(InterpolationType=Average,RotationOrderForFilter=XZY)
    Child=(Type=Bone,Name="dyn_hair_03_r")
    Parents=((Item=(Type=Bone,Name="spine_05"),Weight=0.5),(Item=(Type=Bone,Name="clavicle_r"),Weight=0.5))
    Weight=1
   )
   n18 = RigVMVariableNode() {variable=Head up alpha}
   > &RigUnit_SetRelativeTransformForItem(
    Child=(Type=Bone,Name="dyn_hair_02_r")
    Parent=(Type=Bone,Name="dyn_hair_02_r")
    Value=(Rotation=(X=0,Y=0,Z=0,W=1),Translation=(X=3,Y=0,Z=-4),Scale3D=(X=1,Y=1,Z=1))
    bParentInitial=False
    bPropagateToChildren=True
    Weight=@n18.Value
   ) {template="Set Relative Transform::Execute(in Child,in Parent,in bParentInitial,in Value,in Weight,in bPropagateToChildren)"}
   > &RigUnit_SetRelativeTransformForItem(
    Child=(Type=Bone,Name="dyn_hair_02_c")
    Parent=(Type=Bone,Name="dyn_hair_02_c")
    Value=(Rotation=(X=0,Y=0,Z=0,W=1),Translation=(X=4,Y=-1,Z=0),Scale3D=(X=1,Y=1,Z=1))
    bParentInitial=False
    bPropagateToChildren=True
    Weight=@n18.Value
   ) {template="Set Relative Transform::Execute(in Child,in Parent,in bParentInitial,in Value,in Weight,in bPropagateToChildren)"}
   > &RigUnit_SetRelativeTransformForItem(
    Child=(Type=Bone,Name="dyn_hair_03_r")
    Parent=(Type=Bone,Name="dyn_hair_03_r")
    Value=(Rotation=(X=0,Y=0,Z=0.2588,W=0.9659),Translation=(X=0,Y=0,Z=0),Scale3D=(X=1,Y=1,Z=1))
    bParentInitial=False
    bPropagateToChildren=True
    Weight=@n18.Value
   ) {template="Set Relative Transform::Execute(in Child,in Parent,in bParentInitial,in Value,in Weight,in bPropagateToChildren)"}
   n19 = &RigVMFunction_Sequence(ExecuteContext=())
   n20 = &RigVMFunction_AlphaInterp(
    InterpSpeedDecreasing=9
    InterpSpeedIncreasing=10
    Scale=1
    bClampResult=false
    Value=&RigVMFunction_MathBoolToFloat(Value=RigVMVariableNode() {variable=TacticalsprintingAndJump}.Value).Result
   )
   &RigUnit_SetRelativeTransformForItem(
    Child=(Type=Bone,Name="MSTR_hair_l")
    Parent=(Type=Bone,Name="MSTR_hair_l")
    Value=(Rotation=(X=0.0862,Y=0.0368,Z=-0.1837,W=0.9785),Translation=(X=0,Y=0,Z=0),Scale3D=(X=1,Y=1,Z=1))
    bParentInitial=False
    bPropagateToChildren=True
    Weight=@n20.Result
    ExecutePin=@n19.A
   ) {template="Set Relative Transform::Execute(in Child,in Parent,in bParentInitial,in Value,in Weight,in bPropagateToChildren)"}
   > &RigUnit_SetRelativeTransformForItem(
    Child=(Type=Bone,Name="MSTR_hair_c")
    Parent=(Type=Bone,Name="MSTR_hair_c")
    Value=(Rotation=(X=0,Y=0,Z=-0.1805,W=0.9836),Translation=(X=0,Y=0,Z=0),Scale3D=(X=1,Y=1,Z=1))
    bParentInitial=False
    bPropagateToChildren=True
    Weight=@n20.Result
   ) {template="Set Relative Transform::Execute(in Child,in Parent,in bParentInitial,in Value,in Weight,in bPropagateToChildren)"}
   n21 = &RigVMFunction_AlphaInterp(
    InterpSpeedDecreasing=9
    InterpSpeedIncreasing=10
    Scale=1
    bClampResult=false
    Value=&RigVMFunction_MathBoolToFloat(Value=RigVMVariableNode() {variable=Sprinting}.Value).Result
   )
   &RigUnit_SetRelativeTransformForItem(
    Child=(Type=Bone,Name="MSTR_hair_l")
    Parent=(Type=Bone,Name="MSTR_hair_l")
    Value=(Rotation=(X=-0.014,Y=0.0717,Z=-0.0357,W=0.9967),Translation=(X=0,Y=0,Z=0),Scale3D=(X=1,Y=1,Z=1))
    bParentInitial=False
    bPropagateToChildren=True
    Weight=@n21.Result
    ExecutePin=@n19.B
   ) {template="Set Relative Transform::Execute(in Child,in Parent,in bParentInitial,in Value,in Weight,in bPropagateToChildren)"}
   > &RigUnit_SetRelativeTransformForItem(
    Child=(Type=Bone,Name="dyn_hair_03_l")
    Parent=(Type=Bone,Name="dyn_hair_03_l")
    Value=(Rotation=(X=-0.094,Y=-0.0551,Z=0.1024,W=-0.9888),Translation=(X=0,Y=0,Z=0),Scale3D=(X=1,Y=1,Z=1))
    bParentInitial=False
    bPropagateToChildren=True
    Weight=@n21.Result
   ) {template="Set Relative Transform::Execute(in Child,in Parent,in bParentInitial,in Value,in Weight,in bPropagateToChildren)"}
   > &RigUnit_SetRelativeTransformForItem(
    Child=(Type=Bone,Name="dyn_hair_04_l")
    Parent=(Type=Bone,Name="dyn_hair_04_l")
    Value=(Rotation=(X=0,Y=0,Z=-0.3412,W=0.94),Translation=(X=0,Y=0,Z=0),Scale3D=(X=1,Y=1,Z=1))
    bParentInitial=False
    bPropagateToChildren=True
    Weight=@n21.Result
   ) {template="Set Relative Transform::Execute(in Child,in Parent,in bParentInitial,in Value,in Weight,in bPropagateToChildren)"}
   n22 = &RigUnit_SphericalPoseReader(
    ActiveRegionScaleFactors=(PositiveWidth=1,NegativeWidth=1.5,PositiveHeight=1.5,NegativeHeight=1)
    ActiveRegionSize=0.1
    Debug=(bDrawDebug=false,bDraw2D=False,bDrawLocalAxes=False,DebugScale=25,DebugSegments=20,DebugThickness=0.25)
    DriverAxis=(X=1,Y=0,Z=0)
    DriverItem=(Type=Bone,Name="upperarm_l")
    ExecutePin=()
    FalloffRegionScaleFactors=(PositiveWidth=1.5,NegativeWidth=1.5,PositiveHeight=1,NegativeHeight=1)
    FalloffSize=0.2
    FlipHeightScaling=False
    FlipWidthScaling=False
    OptionalParentItem=(Type=Bone,Name="None")
    RotationOffset=(X=0,Y=-50,Z=0)
   )
   > n23 = &RigUnit_SphericalPoseReader(
    ActiveRegionScaleFactors=(PositiveWidth=1,NegativeWidth=1,PositiveHeight=1,NegativeHeight=1)
    ActiveRegionSize=0.1
    Debug=(bDrawDebug=false,bDraw2D=False,bDrawLocalAxes=False,DebugScale=25,DebugSegments=20,DebugThickness=0.25)
    DriverAxis=(X=-1,Y=0,Z=0)
    DriverItem=(Type=Bone,Name="upperarm_r")
    FalloffRegionScaleFactors=(PositiveWidth=1,NegativeWidth=1,PositiveHeight=1,NegativeHeight=1)
    FalloffSize=0.2
    FlipHeightScaling=False
    FlipWidthScaling=False
    OptionalParentItem=(Type=Bone,Name="None")
    RotationOffset=(X=-50,Y=20,Z=0)
   )
   n24 = &RigVMFunction_AlphaInterp(
    InterpSpeedDecreasing=3
    InterpSpeedIncreasing=10
    Scale=1
    bClampResult=false
    Value=&RigVMFunction_MathDoubleMul(B=@n22.OutputParam, A=@n23.OutputParam).Result
   )
   > &RigUnit_SetRelativeTransformForItem(
    Child=(Type=Bone,Name="MSTR_hair_r")
    Parent=(Type=Bone,Name="MSTR_hair_r")
    Value=(Rotation=(X=0,Y=0,Z=0,W=1),Translation=(X=-2,Y=0,Z=0),Scale3D=(X=1,Y=1,Z=1))
    bParentInitial=False
    bPropagateToChildren=True
    Weight=@n24.Result
   ) {template="Set Relative Transform::Execute(in Child,in Parent,in bParentInitial,in Value,in Weight,in bPropagateToChildren)"}
   > &RigUnit_SetRelativeTransformForItem(
    Child=(Type=Bone,Name="dyn_hair_02_r")
    Parent=(Type=Bone,Name="dyn_hair_02_r")
    Value=(Rotation=(X=0,Y=0,Z=0,W=1),Translation=(X=0,Y=0,Z=0),Scale3D=(X=1,Y=1,Z=1))
    bParentInitial=False
    bPropagateToChildren=True
    Weight=@n24.Result
   ) {template="Set Relative Transform::Execute(in Child,in Parent,in bParentInitial,in Value,in Weight,in bPropagateToChildren)"}
   > &RigUnit_SetRelativeTransformForItem(
    Child=(Type=Bone,Name="dyn_hair_03_r")
    Parent=(Type=Bone,Name="dyn_hair_03_r")
    Value=(Rotation=(X=0,Y=0,Z=0,W=1),Translation=(X=0,Y=0,Z=0),Scale3D=(X=1,Y=1,Z=1))
    bParentInitial=False
    bPropagateToChildren=True
    Weight=@n24.Result
   ) {template="Set Relative Transform::Execute(in Child,in Parent,in bParentInitial,in Value,in Weight,in bPropagateToChildren)"}
   > &RigUnit_SetRelativeTransformForItem(
    Child=(Type=Bone,Name="MSTR_hair_c")
    Parent=(Type=Bone,Name="MSTR_hair_c")
    Value=(Rotation=(X=0,Y=0,Z=0,W=1),Translation=(X=0,Y=0,Z=0),Scale3D=(X=1,Y=1,Z=1))
    bParentInitial=False
    bPropagateToChildren=True
    Weight=@n24.Result
   ) {template="Set Relative Transform::Execute(in Child,in Parent,in bParentInitial,in Value,in Weight,in bPropagateToChildren)"}
   &RigVMFunction_MathDoubleMul(A=1, B=1)
   &RigUnit_SetRelativeTransformForItem(
    Child=(Type=Bone,Name="MSTR_hair_c")
    ExecutePin=()
    Parent=(Type=Bone,Name="MSTR_hair_c")
    Value=(Rotation=(X=0,Y=0,Z=0,W=1),Translation=(X=0,Y=0,Z=0),Scale3D=(X=1,Y=1,Z=1))
    Weight=1
    bParentInitial=False
    bPropagateToChildren=True
   ) {template="Set Relative Transform::Execute(in Child,in Parent,in bParentInitial,in Value,in Weight,in bPropagateToChildren)"}
   &RigUnit_SetRelativeTransformForItem(
    Child=(Type=Bone,Name="dyn_hair_02_c")
    ExecutePin=()
    Parent=(Type=Bone,Name="dyn_hair_02_c")
    Value=(Rotation=(X=0,Y=0,Z=0.0436,W=0.999),Translation=(X=0,Y=0,Z=0),Scale3D=(X=1,Y=1,Z=1))
    bParentInitial=False
    bPropagateToChildren=True
    Weight=@n8.Result
   ) {template="Set Relative Transform::Execute(in Child,in Parent,in bParentInitial,in Value,in Weight,in bPropagateToChildren)"}
  ]}
]
hierarchy [
 root {loc=<0 0 0>, rot=<0 0 0 1>}
 pelvis {parent=root, loc=<0 -0.5301 98.0571>, rot=<0.0127 0.707 -0.0127 -0.707>}
 spine_01 {parent=pelvis, loc=<6.1208 -0.9328 0>, rot=<0 0 0.034 -0.9994>}
 spine_02 {parent=spine_01, loc=<6.5997 -0.2095 0>, rot=<0 0 0.0238 -0.9997>}
 spine_03 {parent=spine_02, loc=<6.4914 -0.1237 0>, rot=<0 0 0.0149 -0.9999>}
 spine_04 {parent=spine_03, loc=<6.9598 0.0794 0>, rot=<0 0 -0.0132 -0.9999>}
 spine_05 {parent=spine_04, loc=<6.9987 0 0>, rot=<0 0 -0.0015 -1>}
 neck_01 {parent=spine_05, loc=<11.3138 0.1529 0>, rot=<0 0 -0.2003 -0.9797>}
 neck_02 {parent=neck_01, loc=<4.5627 0 0>, rot=<0 0 0.0504 -0.9987>}
 head {parent=neck_02, loc=<5.3914 0 0>, rot=<0 0 0.115 -0.9934>}
 MSTR_bow_hairband_l {parent=head, loc=<16.0678 3.6887 -0.5102>, rot=<-0.5027 -0.2046 0.308 -0.7814>}
 dyn_bow_hairband_01_l {parent=MSTR_bow_hairband_l, loc=<0 0 0>, rot=<0 0 0 1>}
 dyn_bow_hairband_02_l {parent=dyn_bow_hairband_01_l, loc=<4.6842 0 0>, rot=<0 0 0 1>}
 MSTR_bow_hairband_r {parent=head, loc=<16.0678 3.6887 0.5102>, rot=<-0.783 0.3073 -0.2055 -0.5003>}
 dyn_bow_hairband_01_r {parent=MSTR_bow_hairband_r, loc=<0 0 0>, rot=<0 0 0 1>}
 dyn_bow_hairband_02_r {parent=dyn_bow_hairband_01_r, loc=<4.6845 0 0>, rot=<0 0 0 1>}
 MSTR_lock_l {parent=head, loc=<-1.6819 1.7784 -8.4804>, rot=<0.1262 0.6958 0.707 -0.0131>}
 dyn_lock_01_l {parent=MSTR_lock_l, loc=<0 0 0>, rot=<0 0 0 1>}
 dyn_lock_02_l {parent=dyn_lock_01_l, loc=<7.0506 0 0>, rot=<0.0195 0.1459 -0.1311 -0.9804>}
 MSTR_lock_r {parent=head, loc=<-1.682 1.7784 8.4804>, rot=<0.0131 -0.707 -0.6958 -0.1262>}
 dyn_lock_01_r {parent=MSTR_lock_r, loc=<0 0 0>, rot=<0 0 0 1>}
 dyn_lock_02_r {parent=dyn_lock_01_r, loc=<7.0503 0 0>, rot=<-0.0276 0.1588 0.1489 -0.9756>}
 MSTR_hair_l {parent=head, loc=<8.6152 -5.2648 -7.6415>, rot=<0.0346 -0.9994 0 0>}
 dyn_hair_01_l {parent=MSTR_hair_l, loc=<0 0 0>, rot=<0 0 0 1>}
 dyn_hair_02_l {parent=dyn_hair_01_l, loc=<8.7235 0 0>, rot=<0 0 0.1687 -0.9857>}
 dyn_hair_03_l {parent=dyn_hair_02_l, loc=<9.3412 0 0>, rot=<0 0 0.0699 -0.9976>}
 dyn_hair_04_l {parent=dyn_hair_03_l, loc=<8.2181 0 0>, rot=<0 0 0 1>}
 MSTR_hair_r {parent=head, loc=<8.6156 -5.2648 7.6415>, rot=<0.0346 -0.9994 0 0>}
 dyn_hair_01_r {parent=MSTR_hair_r, loc=<0 0 0>, rot=<0 0 0 1>}
 dyn_hair_02_r {parent=dyn_hair_01_r, loc=<8.7242 0 0>, rot=<0 0 0.1687 -0.9857>}
 dyn_hair_03_r {parent=dyn_hair_02_r, loc=<9.3413 0.0001 0>, rot=<0 0 0.0699 -0.9976>}
 dyn_hair_04_r {parent=dyn_hair_03_r, loc=<8.218 -0.0001 0>, rot=<0 0 0 -1>}
 MSTR_hair_c {parent=head, loc=<8.6152 -5.2648 0>, rot=<0.0346 -0.9994 0 0>}
 dyn_hair_01_c {parent=MSTR_hair_c, loc=<0 0 0>, rot=<0 0 0 1>}
 dyn_hair_02_c {parent=dyn_hair_01_c, loc=<8.7235 0 0>, rot=<0 0 0.1687 -0.9857>}
 dyn_hair_03_c {parent=dyn_hair_02_c, loc=<9.3412 0 0>, rot=<0 0 0.0699 -0.9976>}
 dyn_hair_04_c {parent=dyn_hair_03_c, loc=<8.2181 0 0>, rot=<0 0 0 1>}
 joint1 {parent=head, loc=<9.1178 9.3884 1.458>, rot=<0.0041 -1 0 0>}
 dyn_bang {parent=head, loc=<9.1178 9.3884 1.458>, rot=<0.0041 -1 0 0>}
 dyn_bang_l {parent=head, loc=<8.6808 6.6508 -7.374>, rot=<0.101 -0.8226 0.5578 -0.0442>}
 clavicle_l {parent=spine_05, loc=<5.4831 -0.1053 -3.2946>, rot=<0.0465 -0.7289 -0.009 -0.6829>}
 upperarm_l {parent=clavicle_l, loc=<9.8658 -1.2136 -0.4105>, rot=<0.0164 -0.3245 -0.0269 -0.9454>}
 lowerarm_l {parent=upperarm_l, loc=<26.465 -0.8345 0.259>, rot=<0.0036 0.0161 -0.2298 -0.9731>}
 upperarm_twist_01_l {parent=upperarm_l, loc=<7.2135 -0.1134 0>, rot=<0 0 0 1>}
 deform_upperarm_bk_base_l {parent=upperarm_twist_01_l, loc=<-6.4053 -4.3746 0.314>, rot=<0 0 0 -1>}
 deform_upperarm_bk_l {parent=deform_upperarm_bk_base_l, loc=<0 0 0>, rot=<0 0 0 1>}
 deform_upperarm_fr_base_l {parent=upperarm_twist_01_l, loc=<-5.6312 5.092 0.3354>, rot=<0 0 0 -1>}
 deform_upperarm_fr_l {parent=deform_upperarm_fr_base_l, loc=<0 0 0>, rot=<0 0 0 1>}
 deform_upperarm_up_base_l {parent=upperarm_twist_01_l, loc=<-8.891 0.4825 3.9517>, rot=<0 0 0 -1>}
 deform_upperarm_up_l {parent=deform_upperarm_up_base_l, loc=<0 0 0>, rot=<0 0 0 1>}
 clavicle_r {parent=spine_05, loc=<5.4831 -0.1053 3.2946>, rot=<0.7289 0.0465 -0.6829 0.009>}
 upperarm_r {parent=clavicle_r, loc=<-9.8658 1.2136 0.4105>, rot=<0.0164 -0.3245 -0.0269 -0.9454>}
 lowerarm_r {parent=upperarm_r, loc=<-26.465 0.8345 -0.259>, rot=<0.0036 0.0161 -0.2298 -0.9731>}
 upperarm_twist_01_r {parent=upperarm_r, loc=<-7.2135 0.1134 0>, rot=<0 0 0 -1>}
 deform_upperarm_bk_base_r {parent=upperarm_twist_01_r, loc=<6.4053 4.3746 -0.314>, rot=<0 0 0 -1>}
 deform_upperarm_bk_r {parent=deform_upperarm_bk_base_r, loc=<0 0 0>, rot=<0 0 0 1>}
 deform_upperarm_fr_base_r {parent=upperarm_twist_01_r, loc=<5.6312 -5.092 -0.3354>, rot=<0 0 0 -1>}
 deform_upperarm_fr_r {parent=deform_upperarm_fr_base_r, loc=<0 0 0>, rot=<0 0 0 1>}
 deform_upperarm_up_base_r {parent=upperarm_twist_01_r, loc=<8.891 -0.4825 -3.9517>, rot=<0 0 0 -1>}
 deform_upperarm_up_r {parent=deform_upperarm_up_base_r, loc=<0 0 0>, rot=<0 0 0 1>}
]
variables [
 Clavicle L Back Alpha {type=float, default=0}
 Clavicle R Back Alpha {type=float, default=0}
 Head up alpha {type=float, default=0}
 Sky dive floating {type=bool, default=False}
 Sky diving {type=bool, default=False}
 Sprinting {type=bool, default=False}
 Swim {type=bool, default=False}
 TacticalsprintingAndJump {type=bool, default=False}
 crouching {type=bool, default=False}
 melee anim {type=bool, default=False}
 parachute opened {type=bool, default=False}
]
