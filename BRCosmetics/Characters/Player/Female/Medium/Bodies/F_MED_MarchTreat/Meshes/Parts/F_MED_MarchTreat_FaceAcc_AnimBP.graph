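# Post-process style Animation Blueprint for the MarchTreat female character's face accessory, driving hair, bangs and hairband bow secondary motion.
# Copied parent and body poses pass through AnimDynamics into a cached pose, then split on backpack or cape.
# The backpack branch runs RigidBody, bone constraints and a Control Rig; the other runs Control Rig, RigidBody and head-look curve offsets.
# The EventGraph derives skydive, sprint, weapon and backbling states for these.
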AnimBlueprint /BRCosmetics/Characters/Player/Female/Medium/Bodies/F_MED_MarchTreat/Meshes/Parts/F_MED_MarchTreat_FaceAcc_AnimBP
target_skeleton=&/F_MED_MarchTreat_FaceAcc_Skeleton
graphs [
 AnimGraph {nodes [
   "Transform (Modify) Bone - Bone: dyn_hair_04_r"(
    ComponentPose="Transform (Modify) Bone - Bone: dyn_hair_02_r"(Alpha=1, Translation="0,0,0", Rotation="0,0,-20", Scale="1,1,1") {rotationMode=BMM_Additive, boneToModify {boneName=dyn_hair_02_r}, rotation {roll=-20}}
    Alpha=1
    Translation="0,0,0"
    Rotation="0,0,30"
    Scale="1,1,1"
   ) {rotationMode=BMM_Additive, boneToModify {boneName=dyn_hair_04_r}, rotation {roll=30}}
   n1 = Control Rig!(
    Source=Use cached pose 'PostDynamics'()
    Alpha
    Sky diving and parachute is close:double=Get sky diving and parachute is close() -> (sky diving and parachute is close:double)
    not sky diving:bool=Get not sky diving() -> (not sky diving:bool)
    skydive Diving Mode:bool=Get bIsSkydiveDiveMode()
    IsParachute Opened Alpha:double=Get Is Parachute Opened Alpha() -> (Is Parachute Opened Alpha:double)
    IsPistol:double=Get Is Pistol Alpha() -> (Is Pistol Alpha:float)
    IsRifle:double=Get Is Rifle Alpha() -> (Is Rifle Alpha:float)
    frontend:bool=Get bIsInFrontEnd()
   ) -> (Pose) {controlRigAssetReference {blueprintRigClass=&/F_MED_MarchTreat_FaceAcc_CtrlRig.F_MED_MarchTreat_FaceAcc_CtrlRig_C}}
   n2 = RigidBody!(ComponentPose=Local To Component(LocalPose=@n1), Alpha, OverrideWorldGravity:Vector="0,0,-980") -> (Pose) {
    bClampLinearTranslationLimitToRefPose=true
    bOverrideWorldGravity=true
    overridePhysicsAsset=&/F_MED_MarchTreat_FaceAcc_Physics
    simulationSpace=BaseBoneSpace
    alphaBoolBlend {blendOption=Linear}
    alphaScaleBiasClamp {bInterpResult=true, interpSpeedIncreasing=2}
    baseBoneRef {boneName=head}
    componentAppliedLinearAccClamp=<3000 3000 3000>
    simSpaceSettings {maxAngularAcceleration=40, maxAngularVelocity=8, maxLinearAcceleration=3500, maxLinearVelocity=700, velocityScaleZ=0.7, worldAlpha=0.5}
   }
   n3 = "Transform (Modify) Bone - Bone: MSTR_hair_c"!(
    ComponentPose="Transform (Modify) Bone - Bone: dyn_hair_04_l"!(
     ComponentPose="Transform (Modify) Bone - Bone: MSTR_hair_l"!(ComponentPose=@n2, AlphaCurveName:Name=head_look_left, Translation="0,0,0", Rotation="0,0,-18", Scale="1,1,1") -> (Pose) {alphaInputType=Curve, rotationMode=BMM_Additive, boneToModify {boneName=MSTR_hair_l}, rotation {roll=-18}}
     AlphaCurveName:Name=head_look_left
     Translation="0,0,0"
     Rotation="0,0,10"
     Scale="1,1,1"
    ) -> (Pose) {alphaInputType=Curve, rotationMode=BMM_Additive, boneToModify {boneName=dyn_hair_04_l}, rotation {roll=10}}
    AlphaCurveName:Name=head_look_left
    Translation="0,0,0"
    Rotation="0,0,-10"
    Scale="1,1,1"
   ) -> (Pose) {alphaInputType=Curve, rotationMode=BMM_Additive, boneToModify {boneName=MSTR_hair_c}, rotation {roll=-10}}
   n4 = RigidBody(ComponentPose=Local To Component(LocalPose=Use cached pose 'PostDynamics'()), OverrideWorldGravity="0,0,-980") {
    bClampLinearTranslationLimitToRefPose=true
    bOverrideWorldGravity=true
    overridePhysicsAsset=&/F_MED_MarchTreat_Backpack_FaceAcc_Physics
    simulationSpace=BaseBoneSpace
    alphaBoolBlend {blendOption=Linear}
    alphaScaleBiasClamp {bInterpResult=true, interpSpeedIncreasing=2}
    baseBoneRef {boneName=head}
    componentAppliedLinearAccClamp=<3000 3000 3000>
    simSpaceSettings {
     maxAngularAcceleration=1000
     maxAngularVelocity=3
     maxLinearAcceleration=2000
     maxLinearVelocity=500
     worldAlpha=0.5
     externalLinearDragV=<0.2 0.2>
    }
   }
   n5 = "Constraint - dyn_hair_02_l (PP)"!(ComponentPose=@n4, Alpha=1, ConstraintWeights_0=0.5, ConstraintWeights_1:float=0.5) -> (Pose) {
    alphaBoolBlend {blendOption=Linear}
    boneToModify {boneName=dyn_hair_02_l}
    constraintSetup [{targetBone {boneName=spine_05}}, {targetBone {boneName=neck_02}}]
    constraintWeights [1,1]
   }
   n6 = "Constraint - dyn_hair_03_l (PP)"!(ComponentPose=@n5, Alpha=1, ConstraintWeights_0=0.5, ConstraintWeights_1:float=0.5) -> (Pose) {
    alphaBoolBlend {blendOption=Linear}
    boneToModify {boneName=dyn_hair_03_l}
    constraintSetup [{targetBone {boneName=spine_05}}, {targetBone {boneName=clavicle_l}}]
    constraintWeights [1,1]
   }
   n7 = "Constraint - dyn_hair_04_l (PP)"!(ComponentPose=@n6, Alpha=1, ConstraintWeights_0=0.5, ConstraintWeights_1:float=0.5) -> (Pose) {
    alphaBoolBlend {blendOption=Linear}
    boneToModify {boneName=dyn_hair_04_l}
    constraintSetup [{targetBone {boneName=spine_05}}, {targetBone {boneName=clavicle_l}}]
    constraintWeights [1,1]
   }
   n8 = "Constraint - dyn_hair_02_c (PP)"!(ComponentPose=@n7, Alpha=1, ConstraintWeights_0=0.5, ConstraintWeights_1:float=0.5) -> (Pose) {
    alphaBoolBlend {blendOption=Linear}
    boneToModify {boneName=dyn_hair_02_c}
    constraintSetup [{targetBone {boneName=spine_05}}, {targetBone {boneName=neck_02}}]
    constraintWeights [1,1]
   }
   n9 = "Constraint - dyn_hair_03_c (PP)"!(ComponentPose=@n8, Alpha=1, ConstraintWeights_0=0.5, ConstraintWeights_1:float=0.5) -> (Pose) {
    alphaBoolBlend {blendOption=Linear}
    boneToModify {boneName=dyn_hair_03_c}
    constraintSetup [{targetBone {boneName=spine_05}}, {targetBone {boneName=neck_01}}]
    constraintWeights [1,1]
   }
   n10 = "Constraint - dyn_hair_04_l (PP)"(ComponentPose=@n9, Alpha=1, ConstraintWeights_0=1, ConstraintWeights_1=0) {
    alphaBoolBlend {blendOption=Linear}
    boneToModify {boneName=dyn_hair_04_l}
    constraintSetup [{targetBone {boneName=spine_05}}, {targetBone {boneName=clavicle_l}}]
    constraintWeights [1,1]
   }
   n11 = "Constraint - dyn_hair_02_r (PP)"!(ComponentPose=@n10, Alpha=1, ConstraintWeights_0=0.5, ConstraintWeights_1:float=0.5) -> (Pose) {
    alphaBoolBlend {blendOption=Linear}
    boneToModify {boneName=dyn_hair_02_r}
    constraintSetup [{targetBone {boneName=head}}, {targetBone {boneName=spine_05}}]
    constraintWeights [1,1]
   }
   n12 = "Constraint - dyn_hair_03_r (PP)"!(ComponentPose=@n11, Alpha=1, ConstraintWeights_0=0.5, ConstraintWeights_1:float=0.5) -> (Pose) {
    alphaBoolBlend {blendOption=Linear}
    boneToModify {boneName=dyn_hair_03_r}
    constraintSetup [{targetBone {boneName=spine_05}}, {targetBone {boneName=clavicle_r}}]
    constraintWeights [1,1]
   }
   n13 = "Constraint - dyn_hair_04_l (PP)"(ComponentPose=@n12, Alpha=1, ConstraintWeights_0=0.5, ConstraintWeights_1=0.5) {
    alphaBoolBlend {blendOption=Linear}
    boneToModify {boneName=dyn_hair_04_l}
    constraintSetup [{targetBone {boneName=spine_05}}, {targetBone {boneName=clavicle_r}}]
    constraintWeights [1,1]
   }
   n14 = Control Rig!(
    Source=Component To Local(ComponentPose="Transform (Modify) Bone - Bone: dyn_hair_03_r"(ComponentPose=@n13, Alpha=1, Translation="0,0,0", Rotation="0,0,-10", Scale="1,1,1") {rotationMode=BMM_Additive, boneToModify {boneName=dyn_hair_03_r}, rotation {roll=-10}})
    Alpha
    parachute opened:bool=Get bIsParachuteOpened()
    Sky dive floating:bool=Get SkyDiveAndFloating() -> (SkyDiveAndFloating:bool)
    Sky diving:bool=Get bIsSkydiveDiveMode()
    melee anim:bool=Get bIsPlayingMeleeAnim()
    TacticalsprintingAndJump:bool=Get TacticalSprintingAndJump() -> (TacticalSprintingAndJump:bool)
    Sprinting:bool=Get Just sprinting() -> (Just sprinting:bool)
    crouching:bool=Get bIsCrouching()
    Swim:bool=Get bIsSurfaceSwimming()
   ) -> (Pose) {controlRigAssetReference {blueprintRigClass=&/F_MED_MarchTreat_FaceAcc_backpack_CtrlRig.F_MED_MarchTreat_FaceAcc_backpack_CtrlRig_C}}
   Output Pose(Result=Blend Poses by bool(
     BlendPose_0=@n14
     BlendPose_1=Component To Local(ComponentPose=@n3)
     bActiveValue=Get IsWearingBackpackOrCape() -> (IsWearingBackpackOrCape:bool)
    ) {blendType=Linear}) {layerGroup=None}
   n15 = Layered blend per bone(
    BasePose=Copy Pose From Mesh() {bCopyCurves=true, bUseAttachedParent=true}
    BlendPoses_0=Copy Pose From Mesh(SourceMeshComponent=Get Body_Mesh()) {bUseAttachedParent=true}
   ) {layerSetup [{branchFilters [
       {boneName=clavicle_l}
       {boneName=clavicle_r}
       {boneName=upperarm_l}
       {boneName=upperarm_r}
       {boneName=upperarm_twist_01_l}
       {boneName=upperarm_twist_01_r}
      ]}]}
   Save cached pose 'StartCache'(Pose=@n15) {cacheName=StartCache}
   n16 = Anim Dynamics - Bone: dyn_bow_hairband_01_l!(
    ComponentPose=Local To Component(LocalPose=Use cached pose 'StartCache'())
    Alpha=1
    GravityOverride:Vector="0,0,-980"
    SimSpaceSettings:Anim Phys Sim Space Settings Structure=(SimSpaceAngularAlpha=0,MaxAngularVelocity=10000,MaxAngularAcceleration=10000,ExternalAngularVelocity=(X=0,Y=0,Z=0))
    ExternalForce="0,0,0"
   ) -> (Pose) {
    angularDampingOverride=1
    angularSpringConstant=5000
    bAngularSpring=true
    bOverrideAngularDamping=true
    bUseGravityOverride=true
    gravityScale=0.35
    simulationSpace=RootRelative
    boundBone {boneName=dyn_bow_hairband_01_l}
    gravityOverride {z=-980}
    physicsBodyDefinitions [{
      boundBone {boneName=dyn_bow_hairband_01_l}
      boxExtents=<16 16 16>
      constraintSetup {angularLimitsMax=<15 15 15>, angularLimitsMin=<-15 -15 -5>, angularTarget=<1 0 0>}
      localJointOffset=<-1 0 0>
     }]
   }
   n17 = Anim Dynamics - Bone: dyn_bow_hairband_01_r!(
    ComponentPose=@n16
    Alpha=1
    GravityOverride:Vector="0,0,-980"
    SimSpaceSettings:Anim Phys Sim Space Settings Structure=(SimSpaceAngularAlpha=0,MaxAngularVelocity=10000,MaxAngularAcceleration=10000,ExternalAngularVelocity=(X=0,Y=0,Z=0))
    ExternalForce="0,0,0"
   ) -> (Pose) {
    angularDampingOverride=1
    angularSpringConstant=5000
    bAngularSpring=true
    bOverrideAngularDamping=true
    bUseGravityOverride=true
    gravityScale=0.35
    simulationSpace=RootRelative
    boundBone {boneName=dyn_bow_hairband_01_r}
    gravityOverride {z=-980}
    physicsBodyDefinitions [{
      boundBone {boneName=dyn_bow_hairband_01_r}
      boxExtents=<16 16 16>
      constraintSetup {angularLimitsMax=<15 15 15>, angularLimitsMin=<-15 -15 -5>, angularTarget=<1 0 0>}
      localJointOffset=<-1 0 0>
     }]
   }
   n18 = Anim Dynamics - Bone: dyn_bang(ComponentPose=@n17, Alpha=1, ExternalForce="0,0,0") {
    angularDampingOverride=0.999
    angularSpringConstant=4500
    bAngularSpring=true
    bOverrideAngularDamping=true
    bUseGravityOverride=true
    gravityScale=0.35
    simulationSpace=RootRelative
    boundBone {boneName=dyn_bang}
    componentAppliedLinearAccClamp=<3000 3000 3000>
    componentLinearAccScale=<0.9 0.9 0.9>
    componentLinearVelScale=<0.9 0.9 0.9>
    gravityOverride {z=-980}
    physicsBodyDefinitions [{
      boundBone {boneName=dyn_bang}
      boxExtents=<20 20 20>
      constraintSetup {angularLimitsMax=<0 15 20>, angularLimitsMin=<0 -15 -2>, angularTarget=<1 0 0>}
      localJointOffset=<-1 0 0>
     }]
   }
   n19 = Anim Dynamics - Bone: dyn_bang_l(ComponentPose=@n18, Alpha=1, ExternalForce="0,0,0") {
    angularDampingOverride=0.999
    angularSpringConstant=4500
    bAngularSpring=true
    bOverrideAngularDamping=true
    bUseGravityOverride=true
    gravityScale=0.35
    simulationSpace=RootRelative
    boundBone {boneName=dyn_bang_l}
    componentAppliedLinearAccClamp=<3000 3000 3000>
    componentLinearAccScale=<0.9 0.9 0.9>
    componentLinearVelScale=<0.9 0.9 0.9>
    gravityOverride {z=-980}
    physicsBodyDefinitions [{
      boundBone {boneName=dyn_bang_l}
      boxExtents=<20 20 20>
      constraintSetup {angularLimitsMax=<0 15 10>, angularLimitsMin=<0 -10 0>, angularTarget=<1 0 0>}
      localJointOffset=<-1 0 0>
     }]
   }
   Save cached pose 'PostDynamics'(Pose=Component To Local(ComponentPose=@n19)) {cacheName=PostDynamics}
  ]}
 EventGraph {nodes [
   Event BlueprintUpdateAnimation()
   > n1 = Sequence!(execute) -> (then_0, then_1, then_2:Exec, then_3:Exec, then_4:Exec, then_5:Exec, then_6:Exec)
   > n2 = "Gravity Override State Function (Parameter Asset) (Nativized)"(
    CharacterPartAnimInstance=Self-Reference()
    GameTimeInSeconds=GetGameTimeInSeconds()
    Parameters=&/F_MED_MarchTreatCup_Parameter
   )
   > Set GravityOverride(GravityOverride=@n2)
   Set not sky diving(execute=@n1.then_2, not sky diving:bool=NOT Boolean(A=Get bIsSkyDiving())) -> (Output_Get:bool)
   n3 = NOT Boolean(A=Get bIsParachuteOpened())
   n4 = Get bIsSkyDiving()
   > Set sky diving and parachute is close(sky diving and parachute is close:double="To Float (Boolean)"(InBool=AND Boolean(A=@n4, B=@n3))) -> (Output_Get:double)
   > Set Is Parachute Opened Alpha(Is Parachute Opened Alpha:double="To Float (Boolean)"(InBool=Get bIsParachuteOpened())) -> (Output_Get:double)
   n5 = Get bIsParachuteOpened()
   n6 = Get bIsSkyDiving()
   n7 = Get bIsSkydiveFloating()
   Set SkyDiveAndFloating(execute=@n1.then_3, SkyDiveAndFloating:bool=AND Boolean(A=NOT Boolean(A=@n5), B=OR Boolean(A=@n6, B=@n7))) -> (Output_Get:bool)
   n8 = Get bIsFalling()
   n9 = Get bIsJumping()
   n10 = Get bIsTacticalSprinting()
   Set TacticalSprintingAndJump(execute=@n1.then_4, TacticalSprintingAndJump:bool=AND Boolean(A=@n10, B=OR Boolean(A=@n8, B=@n9))) -> (Output_Get:bool)
   n11 = NOT Boolean(A=@n9)
   > Set Just sprinting(Just sprinting:bool=AND Boolean(A=AND Boolean(A=@n10, B=@n11), B=NOT Boolean(A=@n8))) -> (Output_Get:bool)
   n12 = Cast To FortPlayerAnimInstance(execute=@n1.then_5, Object=GetAnimInstance(self=GetOwnerSkeletalMeshComponent()))
   > Set isPistol(isPistol=false)
   > Set isRifle(isRifle:bool=false) -> (Output_Get:bool)
   > n13 = Switch on EFortWeaponCoreAnimation!(execute, Selection:EFortWeaponCoreAnimation Enum=Get WeaponCoreAnim(self=@n12)) -> (
    Melee:Exec
    Pistol:Exec
    Shotgun:Exec
    PaperBlueprint:Exec
    Rifle:Exec
    MeleeOneHand:Exec
    MachinePistol:Exec
    RocketLauncher:Exec
    GrenadeLauncher:Exec
    GoingCommando:Exec
    AssaultRifle:Exec
    TacticalShotgun:Exec
    SniperRifle:Exec
    TrapPlacement:Exec
    ShoulderLauncher:Exec
    AbilityDecoTool:Exec
    Crossbow:Exec
    C4:Exec
    RemoteControl:Exec
    DualWield:Exec
    AR_BullPup:Exec
    AR_ForwardGrip:Exec
    MedPackPaddles:Exec
    SMG_P90:Exec
    AR_DrumGun:Exec
    Consumable_Small:Exec
    Consumable_Large:Exec
    Balloon:Exec
    MountedTurret:Exec
    CreativeTool:Exec
    ExplosiveBow:Exec
    AshtonIndigo:Exec
    AshtonChicago:Exec
    MeleeDualWield:Exec
    Unarmed:Exec
   )
   Set isRifle(execute=[@n13.AssaultRifle, @n13.Rifle, @n13.RocketLauncher, @n13.Shotgun, @n13.SniperRifle, @n13.TacticalShotgun], isRifle=true)
   Set isPistol(execute=@n13.Pistol, isPistol=true)
   Set Is Rifle Alpha(execute=@n1.then_6, Is Rifle Alpha:float="To Float (Boolean)"(InBool=Get isRifle() -> (isRifle:bool))) -> (Output_Get:float)
   > Set Is Pistol Alpha(Is Pistol Alpha:float="To Float (Boolean)"(InBool=Get isPistol() -> (isPistol:bool))) -> (Output_Get:float)
   n14 = Event OnCharacterCustomizationLoaded()
   > Set Body_Mesh(Body_Mesh=Get Array Item with EFort Custom Part Type(Array=@n14.CustomMeshParts))
   > Get Dynamic Material Instance(SkeletalMeshComponent=Get Array Item with EFort Custom Part Type(Array=@n14.CustomMeshParts, PartType=Face))
   n15 = CheckForBackbling()
   > Set IsWearingBackpackOrCape(IsWearingBackpackOrCape:bool=OR Boolean!(A=@n15, B=@n15.bIsBackblingACape, C:bool=@n15.bIsBackblingLong) -> (ReturnValue)) -> (Output_Get:bool)
  ]}
]
variables [
 Body_Mesh {type=SkeletalMeshComponent}
 GravityOverride {type=Vector}
 Is Parachute Opened Alpha {type=number}
 Is Pistol Alpha {type=number}
 Is Rifle Alpha {type=number}
 isPistol {type=boolean}
 isRifle {type=boolean}
 IsWearingBackpackOrCape {type=boolean}
 Just sprinting {type=boolean}
 not sky diving {type=boolean}
 sky diving and parachute is close {type=number}
 SkyDiveAndFloating {type=boolean}
 TacticalSprintingAndJump {type=boolean}
]
